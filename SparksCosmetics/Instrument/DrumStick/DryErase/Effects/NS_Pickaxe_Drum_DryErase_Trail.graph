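# Niagara swing trail for the DryErase drumstick pickaxe cosmetic, drawing short-lived black and pink ribbons that follow animation trails.
# One emitter, Trail_Black, spawns two anim trails, using spawn-group masks to vary width, color and opacity.
# WidthCheatBlack and WidthCheatPink attributes zero out the other group's width, apparently separating the colors across two ribbon renderers.
# Alpha fades over lifetime, with brightness and tint adjusted by time of day.
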NiagaraSystem /SparksCosmetics/Instrument/DrumStick/DryErase/Effects/NS_Pickaxe_Drum_DryErase_Trail
dependencies {
 usedDataInterfaces [&NiagaraDataInterfaceCurve, &NiagaraDataInterfaceSocketReader]
 usedDynamicInputs []
 usedModules [
  &SystemState
  &self:SystemSpawnScript.SystemScriptSource.SystemScriptGraph.NiagaraNodeAssignment_0.SetVariables_#1
 ]
}
emitters {Trail_Black {
  emitterUpdateScript [
   &EmitterState()
   &FN_SpawnAnimTrail_v2(Number of Anim Trails=2)
  ]
  particleSpawnScript [
   /Game/Effects/Niagara/Modules/AnimTrail/FN_InitializeAnimTrail/FN_InitializeAnimTrailBoneIndecies(First Bone or Socket Index=1)
   &FN_InitializeAnimTrailBoneIndecies(AnimTrail Selection=1, First Bone or Socket Index=1) {module=FN_InitializeAnimTrailBoneIndecies001}
   /Game/Effects/Niagara/Modules/AnimTrail/FN_InitializeAnimTrail/FN_InitializeAnimTrail(
    AnimTrailWidth:NiagaraFloat=&MaskFloatBySpawnGroup(Masked Float=1.5, Passthrough Float=1.3)
    Color=&MaskLinearColorBySpawnGroup(Masked Linear Color=<rgba 0 0 0 0.5>, Passthrough Linear Color=<rgba 0.9217 0.4697 0.519 0.15>)
    Lifetime=0.15
    Position Offset:Vector3f=&RotateVectorByQuaternion(Vector To Rotate=<0 0 -10>, Quaternion=$Engine.Owner.Rotation)
   )
   &SystemLocation(MaskSpawnGroup=false)
   &AddVelocity(Velocity=<0 400 0>) {enabled=false}
   /SetVariables(
    Particles.WidthCheatBlack:NiagaraFloat=&MaskFloatBySpawnGroup(Masked Float=$Particles.RibbonWidth, Passthrough Float=0)
    Particles.WidthCheatPink:NiagaraFloat=&MaskFloatBySpawnGroup(Masked Float=0, Passthrough Float=$Particles.RibbonWidth)
   ) {isSetParameters=true}
  ]
  particleUpdateScript [
   &ScaleRibbonWidth(Ribbon Width Scale=&FloatFromCurve(FloatCurve=DataInterface() {propertyValues {Curve {keys [
         {arriveTangent=-1.2579, interpMode=RCIM_Cubic, time=0, value=1}
         {arriveTangent=-0.9664, interpMode=RCIM_Cubic, leaveTangent=-0.9664, tangentMode=RCTM_User, time=1, value=0}
        ]}}})) {enabled=false}
   &ScaleColor(
    Scale Alpha=&Multiply_Float(
     A=&FloatFromCurve(FloatCurve=DataInterface() {propertyValues {Curve {keys [
          {interpMode=RCIM_Cubic, leaveTangent=-2.4504, tangentMode=RCTM_Break, time=0, value=1}
          {arriveTangent=-0.2139, interpMode=RCIM_Cubic, tangentMode=RCTM_Break, time=1, value=0}
         ]}}})
     B=&FNTimeOfDayScalar(Night Scale=0.5, Morning Scale=0.65, Day Scale=0.7, Evening Scale=0.7)
    )
    Scale RGB=&FN_Select_Vector_from_Time_Of_Day(Morning=<0.9 0.9 0.9>, Day=<1 1 1>, Evening=<1 1 1>, Night=<0.5 0.5 1>)
    ScaleA=true
    ScaleRGB=true
   ) {module=ScaleColor002}
   &FN_UpdateAnimTrail_v2()
   &SolveForcesAndVelocity(Clamp Velocity=false, Limit Acceleration=false)
  ]
  rendererClasses [&NiagaraRibbonRendererProperties]
  renderers [
   {rendererClass=&NiagaraRibbonRendererProperties}
   {rendererClass=&NiagaraRibbonRendererProperties, rendererIndex=1}
  ]
 }}
summary {
 systemName=NS_Pickaxe_Drum_DryErase_Trail
 emitters [{emitterName=Trail_Black, rendererClasses [&NiagaraRibbonRendererProperties]}]
 userVariables [
  User.RibbonIDOffset {type=NiagaraInt32}
  User.SkeletalMesh {
   type=NiagaraDataInterfaceSkeletalMesh
   defaultValue {
    struct=NiagaraExt_VariableValue_DataInterface
    value {dataInterface=&self:NiagaraDataInterfaceSkeletalMesh_0}
   }
  }
 ]
}
